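# GliderOpen_ChickenWarriorMale_Cue is the sound cue for opening the male Chicken Warrior glider cosmetic.
# A mixer plays the main open sound alongside a slightly quieter caw layer.
# The caw is delayed by a random one to two seconds and passes through a modulator for pitch variation.
# Glider attenuation, a cosmetic property sheet and owner-limited stop-oldest concurrency govern playback, and the cue primes on load.
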SoundCue /BRCosmetics/Sounds/Gliders/ChickenWarrior/GliderOpen_ChickenWarriorMale_Cue
root=Mixer_0
nodes [
 {id=Delay_0, children [WavePlayer_1], properties {delayMax=2, delayMin=1}}
 {id=Mixer_0, children [WavePlayer_2,Modulator_1], properties {inputVolume [1,0.8]}}
 {id=Modulator_1, children [Delay_0], properties {pitchMax=1.1, volumeMin=1}}
 {id=WavePlayer_1, properties {soundWaveAssetPtr=&/GliderOpen_Chicken_Warrior_Male_Caw}}
 {id=WavePlayer_2, properties {soundWaveAssetPtr=&/GliderOpen_Chicken_Warrior_Male}}
]
properties {
 attenuationSettings=&Fort_Player_Glider_Thrust_Athena_Att
 audioPropertiesSheet=/BRCosmetics/Sounds/Settings/PropertySheets/Glider/APS_Cosmetic_Glider_Open
 bOverrideConcurrency=true
 bPrimeOnLoad=true
 maxDistance=12500
 soundClassObject=&Fort_Gliders_Land
 volumeMultiplier=1
 concurrencyOverrides {bLimitToOwner=true, maxCount=2, resolutionRule=StopOldest}
 concurrencySet [&Athena_GliderOpenClose_Conc]
}
