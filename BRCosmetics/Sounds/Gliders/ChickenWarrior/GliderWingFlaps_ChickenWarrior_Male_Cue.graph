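# Sound cue for the male Chicken Warrior glider's wing flaps.
# A random node picks one of four equally weighted flap waves, then a modulator applies a slight random upward pitch shift.
# Playback uses the shared glider thrust attenuation, the glider landing sound class and reduced volume, and primes on load.
# A concurrency override caps simultaneous plays per owner, stopping the oldest, in place of the assigned glider open/close concurrency set.
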SoundCue /BRCosmetics/Sounds/Gliders/ChickenWarrior/GliderWingFlaps_ChickenWarrior_Male_Cue
root=Modulator_0
nodes [
 {id=Modulator_0, children [Random_2], properties {pitchMax=1.2, pitchMin=1.1, volumeMin=1}}
 {id=Random_2, children [WavePlayer_12, WavePlayer_13, WavePlayer_14, WavePlayer_15], properties {weights [1,1,1,1]}}
 {id=WavePlayer_12, properties {soundWaveAssetPtr=&/GliderWingFlaps_ChickenWarrior_Male_03}}
 {id=WavePlayer_13, properties {soundWaveAssetPtr=&/GliderWingFlaps_ChickenWarrior_Male_01}}
 {id=WavePlayer_14, properties {soundWaveAssetPtr=&/GliderWingFlaps_ChickenWarrior_Male_02}}
 {id=WavePlayer_15, properties {soundWaveAssetPtr=&/GliderWingFlaps_ChickenWarrior_Male_04}}
]
properties {
 attenuationSettings=&Fort_Player_Glider_Thrust_Athena_Att
 bOverrideConcurrency=true
 bPrimeOnLoad=true
 maxDistance=12500
 soundClassObject=&Fort_Gliders_Land
 volumeMultiplier=0.3
 concurrencyOverrides {bLimitToOwner=true, maxCount=2, resolutionRule=StopOldest}
 concurrencySet [&Athena_GliderOpenClose_Conc]
}
